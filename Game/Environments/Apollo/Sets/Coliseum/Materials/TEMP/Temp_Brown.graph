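# Temporary placeholder material for the Apollo Coliseum environment set, named for a brown color.
# The graph holds a single constant dark brown color vector, with no textures, parameters or visible link to a material output.
# It uses Default Lit shading and is flagged for use with Nanite meshes.
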Material /Game/Environments/Apollo/Sets/Coliseum/Materials/TEMP/Temp_Brown
graph {
 properties {bUsedWithNanite=true, pixelDepthOffsetMode=PDOM_Legacy, usedShadingModels=Default Lit}
 nodes [Constant3Vector() {constant=<rgba 0.089 0.0657 0.055 0>}]
}
